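# Physics asset for the scarf mesh of the Impulse medium male player body, covering dangling scarf and bag-strap bone chains.
# Capsules on pelvis, spine, thighs and left calf form the body, linked by free-rotation constraints.
# Two four-segment scarf tails hang from spine_05 through master bones; three-segment straps hang from each thigh, front and back on the right via a bag bone.
# Master links rotate freely, while chain segments have tight swing and twist limits.
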PhysicsAsset /Game/Characters/Player/Male/Medium/Bodies/M_MED_Impulse/Meshes/M_MED_Impulse_Scarf_Physics
bodies {
 calf_l [{shapeType=Capsule, length=24, radius=9, shapeName=calf_l_capsule, center=<-15 -2.2119 -0.9176>, rotation=<rot 90 0 0>}]
 dyn_bag_01 [{shapeType=Capsule, length=0.505, radius=0.505, shapeName=dyn_bag_01_capsule, rotation=<rot 0 -90 0>}]
 dyn_bag_MSTR [{shapeType=Sphere, radius=0.505, shapeName=dyn_bag_MSTR_sphere}]
 dyn_bag_l_strap_01 [{shapeType=Capsule, length=3, radius=6, shapeName=dyn_bag_l_strap_01_capsule, center=<4 0 5>, rotation=<rot 90 0 0>}]
 dyn_bag_l_strap_02 [{shapeType=Capsule, length=3, radius=6, shapeName=dyn_bag_l_strap_02_capsule, center=<4 0 5>, rotation=<rot 90 0 0>}]
 dyn_bag_l_strap_03 [{shapeType=Capsule, length=3, radius=6, shapeName=dyn_bag_l_strap_03_capsule, center=<4 0 5>, rotation=<rot 90 0 0>}]
 dyn_bag_l_strap_MSTR [{shapeType=Sphere, radius=0.505, shapeName=dyn_bag_l_strap_MSTR_sphere}]
 dyn_bag_r_strap_b_01 [{shapeType=Capsule, length=3, radius=6, shapeName=dyn_bag_r_strap_b_01_capsule, center=<2 0 -5>, rotation=<rot 90 0 0>}]
 dyn_bag_r_strap_b_02 [{shapeType=Capsule, length=3, radius=6, shapeName=dyn_bag_r_strap_b_02_capsule, center=<2 0 -5>, rotation=<rot 90 0 0>}]
 dyn_bag_r_strap_b_03 [{shapeType=Capsule, length=3, radius=6, shapeName=dyn_bag_r_strap_b_03_capsule, center=<2 0 -5>, rotation=<rot 90 0 0>}]
 dyn_bag_r_strap_f_01 [{shapeType=Capsule, length=3, radius=6, shapeName=dyn_bag_r_strap_f_01_capsule, center=<2 0 -5>, rotation=<rot 90 0 0>}]
 dyn_bag_r_strap_f_02 [{shapeType=Capsule, length=3, radius=6, shapeName=dyn_bag_r_strap_f_02_capsule, center=<2 0 -5>, rotation=<rot 90 0 0>}]
 dyn_bag_r_strap_f_03 [{shapeType=Capsule, length=3, radius=6, shapeName=dyn_bag_r_strap_f_03_capsule, center=<2 0 -5>, rotation=<rot 90 0 0>}]
 dyn_scarf_MSTR [{shapeType=Sphere, radius=0.505, shapeName=dyn_scarf_MSTR_sphere}]
 dyn_scarf_l_01 [{shapeType=Capsule, length=3, radius=3.5, shapeName=dyn_scarf_l_01_capsule, center=<7.8133 2 0>}]
 dyn_scarf_l_02 [{shapeType=Capsule, length=3, radius=3.5, shapeName=dyn_scarf_l_02_capsule, center=<7.8133 2 0>}]
 dyn_scarf_l_03 [{shapeType=Capsule, length=3, radius=3.5, shapeName=dyn_scarf_l_03_capsule, center=<7.8133 2 0>}]
 dyn_scarf_l_04 [{shapeType=Capsule, length=3, radius=3.5, shapeName=dyn_scarf_l_04_capsule, center=<7.8133 2 0>}]
 dyn_scarf_l_MSTR [{shapeType=Sphere, radius=0.505, shapeName=dyn_scarf_l_MSTR_sphere}]
 dyn_scarf_r_01 [{shapeType=Capsule, length=4.5, radius=3.5, shapeName=dyn_scarf_r_01_capsule, center=<7.8133 2 0>, rotation=<rot 0 0.0002 30>}]
 dyn_scarf_r_02 [{shapeType=Capsule, length=4.5, radius=3.5, shapeName=dyn_scarf_r_02_capsule, center=<7.8133 2 0>, rotation=<rot 0 0.0002 30>}]
 dyn_scarf_r_03 [{shapeType=Capsule, length=4.5, radius=3.5, shapeName=dyn_scarf_r_03_capsule, center=<7.8133 2 0>, rotation=<rot 0 0.0002 30>}]
 dyn_scarf_r_04 [{shapeType=Capsule, length=4.5, radius=3.5, shapeName=dyn_scarf_r_04_capsule, center=<7.8133 2 0>, rotation=<rot 0 0.0002 30>}]
 dyn_scarf_r_MSTR [{shapeType=Sphere, radius=0.505, shapeName=dyn_scarf_r_MSTR_sphere}]
 pelvis [{shapeType=Capsule, length=10, radius=10, shapeName=spine_01_capsule, center=<2.8437 1.3157 0>}]
 spine_01 [{shapeType=Capsule, length=10, radius=10, shapeName=spine_01_capsule, center=<6 0.3075 0>}]
 spine_04 [{shapeType=Capsule, length=10, radius=12, shapeName=spine_04_capsule, center=<-5 -0.7205 0>}]
 spine_05 [{shapeType=Capsule, length=10, radius=15, shapeName=spine_05_capsule, center=<-7 -3.531 0>}]
 thigh_l [{shapeType=Capsule, length=20, radius=9, shapeName=thigh_r_capsule, center=<-24.8715 2 0>, rotation=<rot 90 0 0>}]
 thigh_r [{shapeType=Capsule, length=20, radius=9, shapeName=thigh_r_capsule, center=<24.8715 -2 0>, rotation=<rot 90 0 0>}]
}
constraints [
 {bone1Name=dyn_scarf_r_04, bone2Name=dyn_scarf_r_03, swing1LimitDegrees=20, swing2LimitDegrees=12, twistLimitDegrees=3}
 {
  bone1Name=dyn_scarf_l_MSTR
  bone2Name=dyn_scarf_MSTR
  swing1LimitDegrees=45
  swing1Motion=Free
  swing2LimitDegrees=45
  swing2Motion=Free
  twistLimitDegrees=45
  twistMotion=Free
 }
 {bone1Name=dyn_scarf_l_01, bone2Name=dyn_scarf_l_MSTR, swing1LimitDegrees=20, swing2LimitDegrees=12, twistLimitDegrees=3}
 {bone1Name=dyn_scarf_l_02, bone2Name=dyn_scarf_l_01, swing1LimitDegrees=20, swing2LimitDegrees=12, twistLimitDegrees=3}
 {bone1Name=dyn_scarf_l_03, bone2Name=dyn_scarf_l_02, swing1LimitDegrees=20, swing2LimitDegrees=12, twistLimitDegrees=3}
 {bone1Name=dyn_scarf_l_04, bone2Name=dyn_scarf_l_03, swing1LimitDegrees=20, swing2LimitDegrees=12, twistLimitDegrees=3}
 {
  bone1Name=dyn_scarf_r_MSTR
  bone2Name=dyn_scarf_MSTR
  swing1LimitDegrees=45
  swing1Motion=Free
  swing2LimitDegrees=45
  swing2Motion=Free
  twistLimitDegrees=45
  twistMotion=Free
 }
 {bone1Name=dyn_scarf_r_01, bone2Name=dyn_scarf_r_MSTR, swing1LimitDegrees=20, swing2LimitDegrees=12, twistLimitDegrees=3}
 {bone1Name=dyn_scarf_r_02, bone2Name=dyn_scarf_r_01, swing1LimitDegrees=20, swing2LimitDegrees=12, twistLimitDegrees=3}
 {bone1Name=dyn_scarf_r_03, bone2Name=dyn_scarf_r_02, swing1LimitDegrees=20, swing2LimitDegrees=12, twistLimitDegrees=3}
 {
  bone1Name=dyn_scarf_MSTR
  bone2Name=spine_05
  swing1LimitDegrees=45
  swing1Motion=Free
  swing2LimitDegrees=45
  swing2Motion=Free
  twistLimitDegrees=45
  twistMotion=Free
 }
 {
  bone1Name=spine_05
  bone2Name=spine_04
  swing1LimitDegrees=45
  swing1Motion=Free
  swing2LimitDegrees=45
  swing2Motion=Free
  twistLimitDegrees=45
  twistMotion=Free
 }
 {
  bone1Name=spine_04
  bone2Name=spine_01
  swing1LimitDegrees=45
  swing1Motion=Free
  swing2LimitDegrees=45
  swing2Motion=Free
  twistLimitDegrees=45
  twistMotion=Free
 }
 {
  bone1Name=spine_01
  bone2Name=pelvis
  swing1LimitDegrees=45
  swing1Motion=Free
  swing2LimitDegrees=45
  swing2Motion=Free
  twistLimitDegrees=45
  twistMotion=Free
 }
 {
  bone1Name=thigh_l
  bone2Name=pelvis
  swing1LimitDegrees=45
  swing1Motion=Free
  swing2LimitDegrees=45
  swing2Motion=Free
  twistLimitDegrees=45
  twistMotion=Free
 }
 {
  bone1Name=thigh_r
  bone2Name=pelvis
  swing1LimitDegrees=45
  swing1Motion=Free
  swing2LimitDegrees=45
  swing2Motion=Free
  twistLimitDegrees=45
  twistMotion=Free
 }
 {
  bone1Name=dyn_bag_01
  bone2Name=dyn_bag_MSTR
  swing1LimitDegrees=45
  swing1Motion=Free
  swing2LimitDegrees=45
  swing2Motion=Free
  twistLimitDegrees=45
  twistMotion=Free
 }
 {bone1Name=dyn_bag_r_strap_b_01, bone2Name=dyn_bag_01, swing1LimitDegrees=8, swing2LimitDegrees=15, twistLimitDegrees=5}
 {bone1Name=dyn_bag_r_strap_b_03, bone2Name=dyn_bag_r_strap_b_02, swing1LimitDegrees=12, swing2LimitDegrees=30, twistLimitDegrees=30}
 {bone1Name=dyn_bag_r_strap_b_02, bone2Name=dyn_bag_r_strap_b_01, swing1LimitDegrees=12, swing2LimitDegrees=30, twistLimitDegrees=30}
 {bone1Name=dyn_bag_r_strap_f_01, bone2Name=dyn_bag_01, swing1LimitDegrees=8, swing2LimitDegrees=15, twistLimitDegrees=5}
 {bone1Name=dyn_bag_r_strap_f_02, bone2Name=dyn_bag_r_strap_f_01, swing1LimitDegrees=12, swing2LimitDegrees=30, twistLimitDegrees=30}
 {bone1Name=dyn_bag_r_strap_f_03, bone2Name=dyn_bag_r_strap_f_02, swing1LimitDegrees=12, swing2LimitDegrees=30, twistLimitDegrees=30}
 {
  bone1Name=dyn_bag_MSTR
  bone2Name=thigh_r
  swing1LimitDegrees=45
  swing1Motion=Free
  swing2LimitDegrees=45
  swing2Motion=Free
  twistLimitDegrees=45
  twistMotion=Free
 }
 {
  bone1Name=calf_l
  bone2Name=thigh_l
  swing1LimitDegrees=45
  swing1Motion=Free
  swing2LimitDegrees=45
  swing2Motion=Free
  twistLimitDegrees=45
  twistMotion=Free
 }
 {
  bone1Name=dyn_bag_l_strap_MSTR
  bone2Name=thigh_l
  swing1LimitDegrees=45
  swing1Motion=Free
  swing2LimitDegrees=45
  swing2Motion=Free
  twistLimitDegrees=45
  twistMotion=Free
 }
 {bone1Name=dyn_bag_l_strap_01, bone2Name=dyn_bag_l_strap_MSTR, swing1LimitDegrees=8, swing2LimitDegrees=15, twistLimitDegrees=5}
 {bone1Name=dyn_bag_l_strap_02, bone2Name=dyn_bag_l_strap_01, swing1LimitDegrees=20, swing2LimitDegrees=30, twistLimitDegrees=30}
 {bone1Name=dyn_bag_l_strap_03, bone2Name=dyn_bag_l_strap_02, swing1LimitDegrees=20, swing2LimitDegrees=30, twistLimitDegrees=30}
]
