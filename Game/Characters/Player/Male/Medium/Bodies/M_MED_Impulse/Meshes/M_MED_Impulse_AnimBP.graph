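# Animation Blueprint for the medium male Impulse player body, which copies the attached parent mesh's pose and adds secondary motion.
# A constraint ties the left bag strap to the thigh twist bone, and a small bone offset adjusts the left scarf before a Control Rig receives the targeting flag.
# RigidBody then simulates the scarf physics asset in pelvis space with damped component motion, followed by a final scarf nudge.
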AnimBlueprint /Game/Characters/Player/Male/Medium/Bodies/M_MED_Impulse/Meshes/M_MED_Impulse_AnimBP
target_skeleton=&/M_MED_Impulse_Skeleton
graphs [
 AnimGraph {nodes [
   n1 = "Constraint - dyn_bag_l_strap_MSTR (P)"(ComponentPose=Local To Component(LocalPose=Copy Pose From Mesh() {bUseAttachedParent=true}), Alpha=1, ConstraintWeights_0=1) {
    alphaBoolBlend {blendOption=Linear}
    boneToModify {boneName=dyn_bag_l_strap_MSTR}
    constraintSetup [{targetBone {boneName=thigh_twist_01_l}}]
    constraintWeights [1]
   }
   n2 = "Transform (Modify) Bone - Bone: dyn_scarf_l_side_fix"(ComponentPose=@n1, Alpha=1, Translation="0.000,-0.000,-1", Rotation="0,0,0", Scale="1,1,1") {
    translationMode=BMM_Additive
    translationSpace=BCS_BoneSpace
    alphaBoolBlend {blendOption=Linear}
    boneToModify {boneName=dyn_scarf_l_side_fix}
    translation {z=-1}
   }
   n3 = Control Rig!(Source=Component To Local(ComponentPose=@n2), Alpha, Targeting:bool=Get bIsTargeting()) -> (Pose) {
    alphaBoolBlend {blendOption=Linear}
    controlRigAssetReference {blueprintRigClass=&/M_MED_Impulse_CtrlRig.M_MED_Impulse_CtrlRig_C}
   }
   n4 = RigidBody(ComponentPose=Local To Component(LocalPose=@n3)) {
    bClampLinearTranslationLimitToRefPose=true
    bOverrideWorldGravity=true
    overridePhysicsAsset=&/M_MED_Impulse_Scarf_Physics
    simulationSpace=BaseBoneSpace
    alphaBoolBlend {blendOption=Linear}
    baseBoneRef {boneName=pelvis}
    componentAppliedLinearAccClamp=<3000 3000 3000>
    componentLinearAccScale=<0.2 0.2 0.4>
    componentLinearVelScale=<0.1 0.1 0.2>
    overrideWorldGravity {z=-980}
    simSpaceSettings {
     maxAngularAcceleration=1000
     maxAngularVelocity=3
     maxLinearAcceleration=700
     maxLinearVelocity=500
     velocityScaleZ=0
     worldAlpha=1
     externalLinearDragV=<0.2 0.2>
    }
   }
   n5 = "Transform (Modify) Bone - Bone: dyn_scarf_l_02"(ComponentPose=@n4, Alpha=1, Translation="0,1,0", Rotation="0,0,0", Scale="1,1,1") {translationMode=BMM_Additive, alphaBoolBlend {blendOption=Linear}, boneToModify {boneName=dyn_scarf_l_02}, translation {y=1}}
   Output Pose(Result=Component To Local(ComponentPose=@n5)) {layerGroup=None}
  ]}
 EventGraph {nodes [Event BlueprintUpdateAnimation()]}
]
variables []
